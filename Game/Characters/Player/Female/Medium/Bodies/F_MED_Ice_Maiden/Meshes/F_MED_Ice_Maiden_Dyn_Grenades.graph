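# Physics asset for the Ice Maiden female medium body, covering dynamic bones for two hanging grenades and the skirt panels.
# Grenade capsules and the two-box front skirt chain hang from the pelvis with angular limits and tight twist.
# Back and under-skirt box chains use limited joints below master bodies.
# Thigh capsules give leg volume; they, the master bodies and side skirt helper bodies use free angular constraints, some cross-linking panels.
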PhysicsAsset /Game/Characters/Player/Female/Medium/Bodies/F_MED_Ice_Maiden/Meshes/F_MED_Ice_Maiden_Dyn_Grenades
bodies {
 dyn_grenade_01 [{shapeType=Capsule, length=2, radius=2, center=<0.4 0 0>, rotation=<rot 90 0 0>}]
 dyn_grenade_02 [{shapeType=Capsule, length=2, radius=2, center=<0.4 0 0>, rotation=<rot 90 0 0>}]
 dyn_skirt_bk_01 [{shapeType=Box, extentX=19.835, extentY=23.863, extentZ=1.7244, center=<9.2741 0 0>}]
 dyn_skirt_bk_02 [{shapeType=Box, extentX=21.5947, extentY=23.6411, extentZ=1.7244, center=<9.7453 0 0>}]
 dyn_skirt_bk_MSTR [{shapeType=Capsule, length=0.505, radius=0.505, rotation=<rot 0 -90 0>}]
 dyn_skirt_bk_l_02 [{shapeType=Capsule, length=1, radius=1, center=<0 0 1.4164>}]
 dyn_skirt_bk_r_02 [{shapeType=Box, extentX=17.5321, extentY=14.5968, extentZ=3.3481, center=<9.9583 -1.9109 0.1754>, rotation=<rot -11.2845 4.2866 -70.0434>}]
 dyn_skirt_bk_r_03 [{shapeType=Capsule, length=0.505, radius=0.505, center=<6.4426 0.264 -0.3861>, rotation=<rot -3.0953 -90 7.1892>}]
 dyn_skirt_bk_under_01 [{shapeType=Box, extentX=16.3134, extentY=23.863, extentZ=1.6322, center=<8.1148 0 0>}]
 dyn_skirt_bk_under_02 [{shapeType=Box, extentX=19.5317, extentY=23.863, extentZ=1.6322, center=<8.1148 0 0>}]
 dyn_skirt_bk_under_MSTR [{shapeType=Capsule, length=1, radius=1}]
 dyn_skirt_fr_01 [{shapeType=Box, extentX=20.0104, extentY=11.2993, extentZ=1.0305, center=<9.907 0 0.0001>}]
 dyn_skirt_fr_02 [{shapeType=Box, extentX=13.6306, extentY=11.1448, extentZ=1.0305, center=<7.0825 0 0.0001>}]
 dyn_skirt_l_03 [{shapeType=Capsule, length=1, radius=1}]
 pelvis [{shapeType=Capsule, length=2, radius=2}]
 thigh_l [{shapeType=Capsule, length=19.1382, radius=8.8618, center=<-22.1112 0.6724 0.139>, rotation=<rot 90 -12.9074 -7.9075>}]
 thigh_r [{shapeType=Capsule, length=19.4633, radius=8.5367, center=<22.1112 -0.3769 -0.139>, rotation=<rot 90 0 0>}]
 thigh_twist_01_l [{shapeType=Capsule, length=1, radius=1}]
 thigh_twist_01_r [{shapeType=Capsule, length=2, radius=2}]
}
constraints [
 {bone1Name=dyn_skirt_fr_02, bone2Name=dyn_skirt_fr_01, swing1LimitDegrees=10, swing2LimitDegrees=45, twistLimitDegrees=5}
 {bone1Name=dyn_grenade_01, bone2Name=pelvis, swing1LimitDegrees=45, swing2LimitDegrees=45, twistLimitDegrees=2}
 {bone1Name=dyn_grenade_02, bone2Name=pelvis, swing1LimitDegrees=45, swing2LimitDegrees=45, twistLimitDegrees=2}
 {bone1Name=dyn_skirt_fr_01, bone2Name=pelvis, swing1LimitDegrees=45, swing2LimitDegrees=45, twistLimitDegrees=5}
 {bone1Name=dyn_skirt_bk_01, bone2Name=dyn_skirt_bk_MSTR, swing1LimitDegrees=25, swing2LimitDegrees=45, twistLimitDegrees=5}
 {bone1Name=dyn_skirt_bk_02, bone2Name=dyn_skirt_bk_01, swing1LimitDegrees=25, swing2LimitDegrees=45, twistLimitDegrees=5}
 {
  bone1Name=thigh_l
  bone2Name=pelvis
  swing1LimitDegrees=45
  swing1Motion=Free
  swing2LimitDegrees=45
  swing2Motion=Free
  twistLimitDegrees=45
  twistMotion=Free
 }
 {
  bone1Name=thigh_r
  bone2Name=pelvis
  swing1LimitDegrees=45
  swing1Motion=Free
  swing2LimitDegrees=45
  swing2Motion=Free
  twistLimitDegrees=45
  twistMotion=Free
 }
 {
  bone1Name=thigh_twist_01_r
  bone2Name=thigh_r
  swing1LimitDegrees=45
  swing1Motion=Free
  swing2LimitDegrees=45
  swing2Motion=Free
  twistLimitDegrees=45
  twistMotion=Free
 }
 {bone1Name=dyn_skirt_bk_under_01, bone2Name=dyn_skirt_bk_under_MSTR, swing1LimitDegrees=80, swing2LimitDegrees=80, twistLimitDegrees=5}
 {bone1Name=dyn_skirt_bk_under_02, bone2Name=dyn_skirt_bk_under_01, swing1LimitDegrees=35, swing2LimitDegrees=45, twistLimitDegrees=5}
 {
  bone1Name=dyn_skirt_bk_r_03
  bone2Name=dyn_skirt_bk_r_02
  swing1LimitDegrees=45
  swing1Motion=Free
  swing2LimitDegrees=45
  swing2Motion=Free
  twistLimitDegrees=45
  twistMotion=Free
 }
 {
  bone1Name=dyn_skirt_bk_r_03
  bone2Name=dyn_skirt_bk_02
  swing1LimitDegrees=45
  swing1Motion=Free
  swing2LimitDegrees=45
  swing2Motion=Free
  twistLimitDegrees=45
  twistMotion=Free
 }
 {
  bone1Name=dyn_skirt_bk_MSTR
  bone2Name=pelvis
  swing1LimitDegrees=45
  swing1Motion=Free
  swing2LimitDegrees=45
  swing2Motion=Free
  twistLimitDegrees=45
  twistMotion=Free
 }
 {
  bone1Name=dyn_skirt_l_03
  bone2Name=thigh_l
  swing1LimitDegrees=45
  swing1Motion=Free
  swing2LimitDegrees=45
  swing2Motion=Free
  twistLimitDegrees=45
  twistMotion=Free
 }
 {
  bone1Name=dyn_skirt_bk_l_02
  bone2Name=thigh_twist_01_l
  swing1LimitDegrees=45
  swing1Motion=Free
  swing2LimitDegrees=45
  swing2Motion=Free
  twistLimitDegrees=45
  twistMotion=Free
 }
 {
  bone1Name=dyn_skirt_bk_l_02
  bone2Name=dyn_skirt_bk_under_02
  swing1LimitDegrees=45
  swing1Motion=Free
  swing2LimitDegrees=45
  swing2Motion=Free
  twistLimitDegrees=45
  twistMotion=Free
 }
 {
  bone1Name=dyn_skirt_bk_under_MSTR
  bone2Name=dyn_skirt_bk_MSTR
  swing1LimitDegrees=45
  swing1Motion=Free
  swing2LimitDegrees=45
  swing2Motion=Free
  twistLimitDegrees=45
  twistMotion=Free
 }
 {
  bone1Name=thigh_twist_01_l
  bone2Name=thigh_l
  swing1LimitDegrees=45
  swing1Motion=Free
  swing2LimitDegrees=45
  swing2Motion=Free
  twistLimitDegrees=45
  twistMotion=Free
 }
 {
  bone1Name=dyn_skirt_bk_r_02
  bone2Name=pelvis
  swing1LimitDegrees=45
  swing1Motion=Free
  swing2LimitDegrees=45
  swing2Motion=Free
  twistLimitDegrees=45
  twistMotion=Free
 }
]
