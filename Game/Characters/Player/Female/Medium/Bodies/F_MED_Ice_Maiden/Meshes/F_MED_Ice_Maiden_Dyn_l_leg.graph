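# Physics asset for the Ice Maiden female medium player body, covering dynamic skirt bones around the left leg.
# Pelvis, thigh, calf and twist bones carry capsules linked by free angular constraints, while the skirt segments are boxes with angular limits.
# Front and back skirt panels form two-bone chains off the thigh twist, alongside a tiny back master capsule on a free joint.
# Side pieces attach to the calf, one as a two-bone chain and two individually.
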PhysicsAsset /Game/Characters/Player/Female/Medium/Bodies/F_MED_Ice_Maiden/Meshes/F_MED_Ice_Maiden_Dyn_l_leg
bodies {
 calf_l [{shapeType=Capsule, length=2, radius=2}]
 calf_twist_02_l [{shapeType=Capsule, length=27.9098, radius=6.2602, center=<3.3616 -1.2854 0.7668>, rotation=<rot -0.3424 95.6055 -86.558>}]
 dyn_skirt_bk_l_01 [{shapeType=Box, extentX=8.2052, extentY=11.3194, extentZ=1.4931, center=<6.2083 -0.0001 0.0002>, rotation=<rot 0 0 -18.1477>}]
 dyn_skirt_bk_l_02 [{shapeType=Box, extentX=11.159, extentY=11.3194, extentZ=2.4808, center=<5.5934 0 0>, rotation=<rot 0 0 -21.605>}]
 dyn_skirt_bk_l_MSTR [{shapeType=Capsule, length=0.505, radius=0.505, rotation=<rot 0 -90 0>}]
 dyn_skirt_fr_l_01 [{shapeType=Box, extentX=11.9428, extentY=11.8925, extentZ=1.5235, center=<6.2392 0.0001 0.0002>}]
 dyn_skirt_fr_l_02 [{shapeType=Box, extentX=8.0891, extentY=11.8925, extentZ=3.3913, center=<4.4372 0 0>}]
 dyn_skirt_l_01 [{shapeType=Box, extentX=8.0891, extentY=7.5, extentZ=1.0305, center=<5.2122 0 0.4033>}]
 dyn_skirt_l_02 [{shapeType=Box, extentX=13.6306, extentY=7.5, extentZ=1.0305, center=<7.0825 0 0.0001>}]
 dyn_skirt_l_03 [{shapeType=Box, extentX=8.0891, extentY=7.5, extentZ=1.4955, center=<4.0519 0.15 -0.4379>, rotation=<rot -13.6086 4.7572 -19.4772>}]
 dyn_skirt_l_04 [{shapeType=Box, extentX=7.0491, extentY=7.5, extentZ=1.0305, center=<3.6514 0 0.0001>}]
 pelvis [{shapeType=Capsule, length=2, radius=2}]
 thigh_l [{shapeType=Capsule, length=24.9093, radius=7.4167, center=<-31.1166 0.652 -0.0003>, rotation=<rot 90 0 -4.3183>}]
 thigh_twist_01_l [{shapeType=Capsule, length=9.6827, radius=1.9631, center=<-0.0713 0.8551 0>, rotation=<rot 90 0 -4.3183>}]
}
constraints [
 {bone1Name=dyn_skirt_l_02, bone2Name=dyn_skirt_l_01, swing1LimitDegrees=5, swing2LimitDegrees=45, twistLimitDegrees=5}
 {bone1Name=dyn_skirt_l_01, bone2Name=calf_l, swing1LimitDegrees=5, swing2LimitDegrees=45, twistLimitDegrees=5}
 {bone1Name=dyn_skirt_bk_l_02, bone2Name=dyn_skirt_bk_l_01, swing1LimitDegrees=60, swing2LimitDegrees=62, twistLimitDegrees=10}
 {bone1Name=dyn_skirt_bk_l_01, bone2Name=thigh_twist_01_l, swing1LimitDegrees=60, swing2LimitDegrees=60, twistLimitDegrees=10}
 {
  bone1Name=calf_l
  bone2Name=thigh_l
  swing1LimitDegrees=45
  swing1Motion=Free
  swing2LimitDegrees=45
  swing2Motion=Free
  twistLimitDegrees=45
  twistMotion=Free
 }
 {
  bone1Name=thigh_twist_01_l
  bone2Name=thigh_l
  swing1LimitDegrees=45
  swing1Motion=Free
  swing2LimitDegrees=45
  swing2Motion=Free
  twistLimitDegrees=45
  twistMotion=Free
 }
 {
  bone1Name=calf_twist_02_l
  bone2Name=calf_l
  swing1LimitDegrees=45
  swing1Motion=Free
  swing2LimitDegrees=45
  swing2Motion=Free
  twistLimitDegrees=45
  twistMotion=Free
 }
 {bone1Name=dyn_skirt_fr_l_01, bone2Name=thigh_twist_01_l, swing1LimitDegrees=45, swing2LimitDegrees=45, twistLimitDegrees=5}
 {bone1Name=dyn_skirt_fr_l_02, bone2Name=dyn_skirt_fr_l_01, swing1LimitDegrees=5, swing2LimitDegrees=45, twistLimitDegrees=5}
 {bone1Name=dyn_skirt_l_04, bone2Name=calf_l, swing1LimitDegrees=5, swing2LimitDegrees=45, twistLimitDegrees=5}
 {bone1Name=dyn_skirt_l_03, bone2Name=calf_l, swing1LimitDegrees=5, swing2LimitDegrees=45, twistLimitDegrees=5}
 {
  bone1Name=thigh_l
  bone2Name=pelvis
  swing1LimitDegrees=45
  swing1Motion=Free
  swing2LimitDegrees=45
  swing2Motion=Free
  twistLimitDegrees=45
  twistMotion=Free
 }
 {
  bone1Name=dyn_skirt_bk_l_MSTR
  bone2Name=thigh_twist_01_l
  swing1LimitDegrees=45
  swing1Motion=Free
  swing2LimitDegrees=45
  swing2Motion=Free
  twistLimitDegrees=45
  twistMotion=Free
 }
]
