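# Physics asset for the Ice Maiden player body's right-leg dynamics, covering the right-side skirt and a knife.
# Capsules on the pelvis, right leg and thigh twists, plus the skirt master and back under-panel, link through free constraints.
# Box skirt panels hang from the master (back, branching after the second panel) and right thigh twist (front) with angular limits.
# The small knife capsule is tightly limited on the right thigh twist.
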PhysicsAsset /Game/Characters/Player/Female/Medium/Bodies/F_MED_Ice_Maiden/Meshes/F_MED_Ice_Maiden_Dyn_r_leg
bodies {
 calf_r [{shapeType=Capsule, length=43.6794, radius=6.5215, center=<25.8378 2.2315 0.9499>, rotation=<rot -0.0038 89.9937 -86.5411>}]
 dyn_knife [{shapeType=Capsule, length=2, radius=2, center=<0.5 0 0>, rotation=<rot 90 0 0>}]
 dyn_skirt_bk_r_01 [{shapeType=Box, extentX=18.1137, extentY=21.1608, extentZ=2.4753, center=<9.3825 0 0.0001>}]
 dyn_skirt_bk_r_02 [{shapeType=Box, extentX=18.8344, extentY=21.1608, extentZ=2.4753, center=<9.4277 0 0>}]
 dyn_skirt_bk_r_03 [{shapeType=Box, extentX=13.6306, extentY=15.7189, extentZ=1.0305, center=<7.0825 0 0.0001>}]
 dyn_skirt_bk_r_04 [{shapeType=Box, extentX=13.6306, extentY=15.7189, extentZ=2.0305, center=<7.0825 0 0.0001>}]
 dyn_skirt_bk_r_05 [{shapeType=Box, extentX=13.6306, extentY=15.8783, extentZ=1.0305, center=<7.0825 0 0.0001>}]
 dyn_skirt_bk_r_MSTR [{shapeType=Capsule, length=0.505, radius=0.505, rotation=<rot 0 -90 0>}]
 dyn_skirt_bk_under_01 [{shapeType=Box, extentX=12.0055, extentY=27.7769, extentZ=2.8034, center=<6.3052 -7.8455 -3.6437>, rotation=<rot 3.4683 -1.2641 -20.0396>}]
 dyn_skirt_fr_r_01 [{shapeType=Box, extentX=13.6306, extentY=12.7338, extentZ=2.0735, center=<7.0825 0 0.0001>}]
 dyn_skirt_fr_r_02 [{shapeType=Box, extentX=13.6306, extentY=12.7338, extentZ=4.9133, center=<7.0825 0 0.0001>, rotation=<rot 0 0 -29.9997>}]
 pelvis [{shapeType=Capsule, length=12.75, radius=7.75}]
 thigh_r [{shapeType=Capsule, length=25.4562, radius=8.6598, center=<24.4512 -1.3004 -0.3418>, rotation=<rot 0.3013 84.9404 -86.5543>}]
 thigh_twist_01_l [{shapeType=Capsule, length=1, radius=1}]
 thigh_twist_01_r [{shapeType=Capsule, length=2, radius=2}]
}
constraints [
 {bone1Name=dyn_skirt_bk_r_02, bone2Name=dyn_skirt_bk_r_01, swing1LimitDegrees=20.4286, swing2LimitDegrees=30, twistLimitDegrees=5}
 {bone1Name=dyn_knife, bone2Name=thigh_twist_01_r, swing1LimitDegrees=3, swing2LimitDegrees=10, twistLimitDegrees=2}
 {bone1Name=dyn_skirt_bk_r_01, bone2Name=dyn_skirt_bk_r_MSTR, swing1LimitDegrees=12.2857, swing2LimitDegrees=30, twistLimitDegrees=5}
 {
  bone1Name=thigh_twist_01_r
  bone2Name=thigh_r
  swing1LimitDegrees=45
  swing1Motion=Free
  swing2LimitDegrees=45
  swing2Motion=Free
  twistLimitDegrees=45
  twistMotion=Free
 }
 {bone1Name=dyn_skirt_fr_r_01, bone2Name=thigh_twist_01_r, swing1LimitDegrees=10, swing2LimitDegrees=45, twistLimitDegrees=5}
 {
  bone1Name=thigh_r
  bone2Name=pelvis
  swing1LimitDegrees=45
  swing1Motion=Free
  swing2LimitDegrees=45
  swing2Motion=Free
  twistLimitDegrees=45
  twistMotion=Free
 }
 {
  bone1Name=calf_r
  bone2Name=thigh_r
  swing1LimitDegrees=45
  swing1Motion=Free
  swing2LimitDegrees=45
  swing2Motion=Free
  twistLimitDegrees=45
  twistMotion=Free
 }
 {bone1Name=dyn_skirt_bk_r_05, bone2Name=dyn_skirt_bk_r_02, swing1LimitDegrees=16.8571, swing2LimitDegrees=45, twistLimitDegrees=5}
 {bone1Name=dyn_skirt_bk_r_03, bone2Name=dyn_skirt_bk_r_02, swing1LimitDegrees=26.8571, swing2LimitDegrees=50, twistLimitDegrees=5}
 {bone1Name=dyn_skirt_bk_r_04, bone2Name=dyn_skirt_bk_r_02, swing1LimitDegrees=20, swing2LimitDegrees=45, twistLimitDegrees=5}
 {bone1Name=dyn_skirt_fr_r_02, bone2Name=dyn_skirt_fr_r_01, swing1LimitDegrees=10, swing2LimitDegrees=45, twistLimitDegrees=5}
 {
  bone1Name=dyn_skirt_bk_under_01
  bone2Name=pelvis
  swing1LimitDegrees=45
  swing1Motion=Free
  swing2LimitDegrees=45
  swing2Motion=Free
  twistLimitDegrees=45
  twistMotion=Free
 }
 {
  bone1Name=thigh_twist_01_l
  bone2Name=pelvis
  swing1LimitDegrees=45
  swing1Motion=Free
  swing2LimitDegrees=45
  swing2Motion=Free
  twistLimitDegrees=45
  twistMotion=Free
 }
 {
  bone1Name=dyn_skirt_bk_r_MSTR
  bone2Name=pelvis
  swing1LimitDegrees=45
  swing1Motion=Free
  swing2LimitDegrees=45
  swing2Motion=Free
  twistLimitDegrees=45
  twistMotion=Free
 }
]
