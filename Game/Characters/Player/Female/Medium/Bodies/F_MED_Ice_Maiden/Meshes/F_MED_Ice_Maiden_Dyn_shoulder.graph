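# Physics asset for the Ice Maiden medium female character body, covering the dynamic left-shoulder fur and flower bones.
# Small capsules on the front, outer and back fur bones are constrained to the left upper-arm twist bone, and the flower to the upper spine.
# These dangling joints have moderate swing limits and very tight twist.
# The supporting chain from arm through spine to pelvis uses free angular constraints.
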PhysicsAsset /Game/Characters/Player/Female/Medium/Bodies/F_MED_Ice_Maiden/Meshes/F_MED_Ice_Maiden_Dyn_shoulder
bodies {
 dyn_flower [{shapeType=Capsule, length=2, radius=2, center=<0.4 0 0>, rotation=<rot 90 0 0>}]
 dyn_shoulder_fur_bk_l [{shapeType=Capsule, length=2, radius=2, center=<0.4 0 0>, rotation=<rot 90 0 0>}]
 dyn_shoulder_fur_fr_l [{shapeType=Capsule, length=2, radius=2, center=<0.4 0 0>, rotation=<rot 90 0 0>}]
 dyn_shoulder_fur_out_l [{shapeType=Capsule, length=2, radius=2, center=<0.4 0 0>, rotation=<rot 90 0 0>}]
 pelvis [{shapeType=Capsule, length=2, radius=2}]
 spine_05 [{shapeType=Capsule, length=2, radius=2}]
 upperarm_twist_01_l [{shapeType=Capsule, length=2, radius=2}]
}
constraints [
 {bone1Name=dyn_shoulder_fur_fr_l, bone2Name=upperarm_twist_01_l, swing1LimitDegrees=45, swing2LimitDegrees=45, twistLimitDegrees=2}
 {bone1Name=dyn_shoulder_fur_out_l, bone2Name=upperarm_twist_01_l, swing1LimitDegrees=45, swing2LimitDegrees=45, twistLimitDegrees=2}
 {bone1Name=dyn_shoulder_fur_bk_l, bone2Name=upperarm_twist_01_l, swing1LimitDegrees=45, swing2LimitDegrees=45, twistLimitDegrees=2}
 {
  bone1Name=upperarm_twist_01_l
  bone2Name=spine_05
  swing1LimitDegrees=45
  swing1Motion=Free
  swing2LimitDegrees=45
  swing2Motion=Free
  twistLimitDegrees=45
  twistMotion=Free
 }
 {bone1Name=dyn_flower, bone2Name=spine_05, swing1LimitDegrees=45, swing2LimitDegrees=45, twistLimitDegrees=2}
 {
  bone1Name=spine_05
  bone2Name=pelvis
  swing1LimitDegrees=45
  swing1Motion=Free
  swing2LimitDegrees=45
  swing2Motion=Free
  twistLimitDegrees=45
  twistMotion=Free
 }
]
