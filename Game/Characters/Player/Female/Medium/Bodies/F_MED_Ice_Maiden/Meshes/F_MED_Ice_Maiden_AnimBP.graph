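# Animation Blueprint for the Ice Maiden female player body, adding secondary motion to skirt, shoulder fur and grenades.
# It copies the attached parent's pose, then applies bone constraints and a Look At before four RigidBody simulations fed by skydiving force.
# A final Look At and back-left skirt bone offset follow, the offset enabled only while crouched and stationary.
# Some constraint, Look At, Flap and Sequence Player nodes sit unconnected to the output.
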AnimBlueprint /Game/Characters/Player/Female/Medium/Bodies/F_MED_Ice_Maiden/Meshes/F_MED_Ice_Maiden_AnimBP
target_skeleton=&/F_MED_Ice_Maiden_Skeleton
graphs [
 AnimGraph {nodes [
   n1 = "Constraint - dyn_skirt_bk_MSTR (PT)"!(ComponentPose, Alpha=1, ConstraintWeights_0=1, ConstraintWeights_1:float=1) -> (Pose) {
    alphaBoolBlend {blendOption=Linear}
    boneToModify {boneName=dyn_skirt_bk_MSTR}
    constraintSetup [{targetBone {boneName=thigh_l}}, {transformType=Translation, targetBone {boneName=thigh_r}}]
    constraintWeights [1,1]
   }
   "Constraint - dyn_skirt_bk_r_01 (P)"(ComponentPose=@n1, Alpha=1, ConstraintWeights_0=0.5) {
    alphaBoolBlend {blendOption=Linear}
    boneToModify {boneName=dyn_skirt_bk_r_01}
    constraintSetup [{targetBone {boneName=thigh_r}}]
    constraintWeights [1]
   }
   Look At - Bone: dyn_skirt_bk_under_MSTR(Alpha=1) {
    alphaBoolBlend {blendOption=Linear}
    boneToModify {boneName=dyn_skirt_bk_under_MSTR}
    lookAtLocation {x=0}
    lookAtTarget {boneReference {boneName=foot_l}}
    lookAt_Axis {axis=<1 0>}
    lookUp_Axis {axis=<0 1>}
   }
   n2 = "Constraint - dyn_skirt_fr_01 (RR)"!(
    ComponentPose=Local To Component(LocalPose=Copy Pose From Mesh() {bUseAttachedParent=true})
    Alpha=1
    ConstraintWeights_0=1
    ConstraintWeights_1:float=1
   ) -> (Pose) {
    alphaBoolBlend {blendOption=Linear}
    boneToModify {boneName=dyn_skirt_fr_01}
    constraintSetup [{transformType=Rotation, targetBone {boneName=thigh_l}}, {transformType=Rotation, targetBone {boneName=thigh_r}}]
    constraintWeights [1,1]
   }
   n3 = "Constraint - dyn_skirt_fr_02 (TT)"!(ComponentPose=@n2, Alpha=1, ConstraintWeights_0=1, ConstraintWeights_1:float=1) -> (Pose) {
    alphaBoolBlend {blendOption=Linear}
    boneToModify {boneName=dyn_skirt_fr_02}
    constraintSetup [{transformType=Translation, targetBone {boneName=dyn_skirt_fr_r_02}}, {transformType=Translation, targetBone {boneName=dyn_skirt_fr_l_02}}]
    constraintWeights [1,1]
   }
   n4 = "Constraint - dyn_skirt_bk_r_MSTR (PPP)"!(ComponentPose=@n3, Alpha=1, ConstraintWeights_0=0.6399, ConstraintWeights_1:float=0.0098, ConstraintWeights_2:float=0.3503) -> (Pose) {
    alphaBoolBlend {blendOption=Linear}
    boneToModify {boneName=dyn_skirt_bk_r_MSTR}
    constraintSetup [{targetBone {boneName=pelvis}}, {targetBone {boneName=thigh_twist_01_l}}, {targetBone {boneName=thigh_twist_01_r}}]
    constraintWeights [1,1,1]
   }
   n5 = "Constraint - dyn_skirt_bk_MSTR (PPP)"!(ComponentPose=@n4, Alpha=1, ConstraintWeights_0=0.6368, ConstraintWeights_1:float=0.1682, ConstraintWeights_2:float=0.195) -> (Pose) {
    alphaBoolBlend {blendOption=Linear}
    boneToModify {boneName=dyn_skirt_bk_MSTR}
    constraintSetup [{targetBone {boneName=pelvis}}, {targetBone {boneName=thigh_twist_01_l}}, {targetBone {boneName=thigh_twist_01_r}}]
    constraintWeights [1,1,1]
   }
   n6 = "Constraint - dyn_shoulder_fur_fr_l (PPPP)"!(
    ComponentPose=@n5
    Alpha=1
    ConstraintWeights_0=0.0602
    ConstraintWeights_1:float=0
    ConstraintWeights_2:float=0.1409
    ConstraintWeights_3:float=0.7988
   ) -> (Pose) {
    alphaBoolBlend {blendOption=Linear}
    boneToModify {boneName=dyn_shoulder_fur_fr_l}
    constraintSetup [
     {targetBone {boneName=upperarm_twist_01_l}}
     {targetBone {boneName=spine_04}}
     {targetBone {boneName=spine_05}}
     {targetBone {boneName=clavicle_l}}
    ]
    constraintWeights [1,1,1,1]
   }
   n7 = "Constraint - dyn_shoulder_fur_bk_l (PP)"!(ComponentPose=@n6, Alpha=1, ConstraintWeights_0=0.3531, ConstraintWeights_1:float=0.6469) -> (Pose) {
    alphaBoolBlend {blendOption=Linear}
    boneToModify {boneName=dyn_shoulder_fur_bk_l}
    constraintSetup [{targetBone {boneName=upperarm_twist_01_l}}, {targetBone {boneName=clavicle_l}}]
    constraintWeights [1,1]
   }
   n8 = "Constraint - dyn_shoulder_fur_out_l (PP)"!(ComponentPose=@n7, Alpha=1, ConstraintWeights_0=0.3873, ConstraintWeights_1:float=0.6127) -> (Pose) {
    alphaBoolBlend {blendOption=Linear}
    boneToModify {boneName=dyn_shoulder_fur_out_l}
    constraintSetup [{targetBone {boneName=upperarm_twist_01_l}}, {targetBone {boneName=clavicle_l}}]
    constraintWeights [1,1]
   }
   n9 = Look At - Bone: dyn_skirt_fr_r_01(ComponentPose=@n8, Alpha=1) {
    alphaBoolBlend {blendOption=Linear}
    boneToModify {boneName=dyn_skirt_fr_r_01}
    lookAtLocation=<0 20>
    lookAtTarget {boneReference {boneName=dyn_skirt_fr_l_01}}
    lookAt_Axis {axis=<1 0>}
   }
   n10 = Break Fort Anim Input Skydiving External Force(FortAnimInput_SkydivingExternalForce=Get SkydivingExternalForceInput()) -> (
    bUseSkydivingVectorForce=~
    SkydivingVectorForce=~
    HeadToPelvisDirection=~
    FloatingMultiplier=~
    FloatingAdditive=~
    DivingMultiplier=~
    DivingAdditive=~
    ParachutingMultiplier=~
    ParachutingAdditive=~
   )
   n11 = RigidBody(ComponentPose=@n9, ExternalForce=@n10.SkydivingVectorForce) {
    bClampLinearTranslationLimitToRefPose=true
    bOverrideWorldGravity=true
    overridePhysicsAsset=&/F_MED_Ice_Maiden_Dyn_shoulder
    simulationSpace=BaseBoneSpace
    alphaBoolBlend {blendOption=Linear}
    baseBoneRef {boneName=spine_05}
    componentAppliedLinearAccClamp=<2000 2000 2000>
    componentLinearAccScale=<1 0.7 0.7>
    componentLinearVelScale=<1 0.7 0.7>
    overrideWorldGravity {z=-980}
   }
   n12 = RigidBody(ComponentPose=@n11, ExternalForce=@n10.SkydivingVectorForce) {
    bClampLinearTranslationLimitToRefPose=true
    bOverrideWorldGravity=true
    overridePhysicsAsset=&/F_MED_Ice_Maiden_Dyn_l_leg
    simulationSpace=BaseBoneSpace
    alphaBoolBlend {blendOption=Linear}
    baseBoneRef {boneName=thigh_l}
    componentAppliedLinearAccClamp=<2000 2000 2000>
    componentLinearAccScale=<0.5 0.5 0.5>
    componentLinearVelScale=<0.5 0.5 0.5>
    overrideWorldGravity {z=-980}
   }
   n13 = RigidBody(ComponentPose=@n12, ExternalForce=@n10.SkydivingVectorForce) {
    bClampLinearTranslationLimitToRefPose=true
    bOverrideWorldGravity=true
    overridePhysicsAsset=&/F_MED_Ice_Maiden_Dyn_Grenades
    simulationSpace=BaseBoneSpace
    alphaBoolBlend {blendOption=Linear}
    baseBoneRef {boneName=pelvis}
    componentAppliedLinearAccClamp=<2000 2000 2000>
    componentLinearAccScale=<0.5 0.5 0.5>
    componentLinearVelScale=<0.5 0.5 0.5>
    overrideWorldGravity {z=-980}
   }
   n14 = RigidBody(ComponentPose=@n13, ExternalForce=@n10.SkydivingVectorForce) {
    bClampLinearTranslationLimitToRefPose=true
    bOverrideWorldGravity=true
    overridePhysicsAsset=&/F_MED_Ice_Maiden_Dyn_r_leg
    simulationSpace=BaseBoneSpace
    alphaBoolBlend {blendOption=Linear}
    baseBoneRef {boneName=root}
    componentAppliedLinearAccClamp=<2000 2000 2000>
    componentLinearAccScale=<0.5 0.5 0.5>
    componentLinearVelScale=<0.5 0.5 0.5>
    overrideWorldGravity {z=-850}
   }
   n15 = Look At - Bone: dyn_skirt_fr_02(ComponentPose=@n14, Alpha=1) {
    bUseLookUpAxis=true
    alphaBoolBlend {blendOption=Linear}
    boneToModify {boneName=dyn_skirt_fr_02}
    lookAtLocation {x=0}
    lookAtTarget {boneReference {boneName=dyn_skirt_fr_l_02}}
    lookAt_Axis {axis=<1 0>}
    lookUp_Axis {axis=<0 1>}
   }
   n16 = "Transform (Modify) Bone - Bone: dyn_skirt_bk_l_MSTR"(
    ComponentPose=@n15
    Alpha=Get crouching_alpha() -> (crouching_alpha:double)
    Translation="-10.513,4.188,1.439"
    Rotation="0,0,0"
    Scale="1,1,1"
   ) {
    rotationMode=BMM_Replace
    rotationSpace=BCS_BoneSpace
    translationMode=BMM_Replace
    translationSpace=BCS_BoneSpace
    alphaBoolBlend {blendOption=Linear}
    boneToModify {boneName=dyn_skirt_bk_l_MSTR}
    translation=<-10.513 4.188 1.439>
   }
   Output Pose(Result=Component To Local(ComponentPose=@n16)) {layerGroup=None}
   Sequence Player 'RunningAround_v06'() {playRate=0, sequence=&/RunningAround_v06}
   Flap Node - Bone: dyn_skirt_fr_01!(ComponentPose:Component Space Pose Link Structure, Alpha:float=1) -> (Pose:Component Space Pose Link Structure) {
    is=title FortAnimGraphNode_Flap
    bUseClamp=true
    targetClampMax=3.14
    targetComponent=RotationZ
    alphaBoolBlend {blendOption=Linear}
    sourceDrives [
     {
      bUseRange=false
      drivingCurve=null
      multiplier=1
      rangeMax=1
      rangeMin=-1
      remappedMax=1
      remappedMin=0
      sourceComponent=RotationZ
      useQuaternion=false
      sourceBone {boneName=thigh_l}
     }
     {
      bUseRange=false
      drivingCurve=null
      multiplier=1
      rangeMax=1
      rangeMin=-1
      remappedMax=1
      remappedMin=0
      sourceComponent=RotationZ
      useQuaternion=false
      sourceBone {boneName=thigh_r}
     }
    ]
    targetBone {boneName=dyn_skirt_fr_01}
   }
  ]}
 EventGraph {nodes [
   Event BlueprintUpdateAnimation()
   > n1 = Branch(Condition=Get bIsCrouching())
   > n2 = Set isMoving(isMoving:bool="float > float"(A=Get PawnSpeed2D(), B=0) {is=title K2Node_CallFunction}) -> (Output_Get:bool)
   > n3 = Branch(Condition=@n2)
   > Set crouching_alpha(crouching_alpha:double=0) -> (Output_Get:double)
   Set crouching_alpha(execute=@n3.else, crouching_alpha=1)
   Set crouching_alpha(execute=@n1.else, crouching_alpha=0)
  ]}
]
variables [
 crouching_alpha {type=number}
 isMoving {type=boolean}
]
